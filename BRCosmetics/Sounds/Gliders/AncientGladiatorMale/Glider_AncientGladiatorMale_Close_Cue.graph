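# Sound cue for the close sound of the AncientGladiatorMale glider cosmetic.
# A single wave player feeds the close sound wave through a modulator node at the root.
# It uses the shared Athena glider open/close attenuation, the cosmetic glider close property sheet and the glider landing sound class, with reduced volume.
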SoundCue /BRCosmetics/Sounds/Gliders/AncientGladiatorMale/Glider_AncientGladiatorMale_Close_Cue
root=Modulator_0
nodes [
 {id=Modulator_0, children [WavePlayer_1]}
 {id=WavePlayer_1, properties {soundWaveAssetPtr=&/Glider_AncientGladiatorMale_Close}}
]
properties {
 attenuationSettings=/Game/Sounds/Attenuation/Fort_Player_Glider_OpenClose_Athena_Att
 audioPropertiesSheet=/BRCosmetics/Sounds/Settings/PropertySheets/Glider/APS_Cosmetic_Glider_Close
 maxDistance=10000
 soundClassObject=&Fort_Gliders_Land
 volumeMultiplier=0.6
}
