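# Sound cue for the opening of the Ancient Gladiator Male glider cosmetic.
# A mixer layers the glider open sound with a quieter voice line that starts after a fixed delay.
# The voice line is picked at random from nine equally weighted VOX takes and pitched up within a random range.
# Glider-specific attenuation, concurrency, sound class and an audio property sheet govern playback.
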SoundCue /BRCosmetics/Sounds/Gliders/AncientGladiatorMale/Glider_AncientGladiatorMale_Open_Cue
root=Mixer_0
nodes [
 {id=Delay_0, children [Modulator_1], properties {delayMax=1.8, delayMin=1.8}}
 {id=Mixer_0, children [WavePlayer_1,Delay_0], properties {inputVolume [1.2,0.6]}}
 {id=Modulator_1, children [Random_1], properties {pitchMax=1.35, pitchMin=1.2}}
 {
  id=Random_1
  children [WavePlayer_0,WavePlayer_2,WavePlayer_3,WavePlayer_4,WavePlayer_5,WavePlayer_6,WavePlayer_7,WavePlayer_8,WavePlayer_9]
  properties {weights [1,1,1,1,1,1,1,1,1]}
 }
 {id=WavePlayer_0, properties {soundWaveAssetPtr=&/Glider_AncientGladiatorMale_VOX_01}}
 {id=WavePlayer_1, properties {soundWaveAssetPtr=&/Glider_AncientGladiatorMale_Open}}
 {id=WavePlayer_2, properties {soundWaveAssetPtr=&/Glider_AncientGladiatorMale_VOX_02}}
 {id=WavePlayer_3, properties {soundWaveAssetPtr=&/Glider_AncientGladiatorMale_VOX_03}}
 {id=WavePlayer_4, properties {soundWaveAssetPtr=&/Glider_AncientGladiatorMale_VOX_04}}
 {id=WavePlayer_5, properties {soundWaveAssetPtr=&/Glider_AncientGladiatorMale_VOX_05}}
 {id=WavePlayer_6, properties {soundWaveAssetPtr=&/Glider_AncientGladiatorMale_VOX_06}}
 {id=WavePlayer_7, properties {soundWaveAssetPtr=&/Glider_AncientGladiatorMale_VOX_07}}
 {id=WavePlayer_8, properties {soundWaveAssetPtr=&/Glider_AncientGladiatorMale_VOX_08}}
 {id=WavePlayer_9, properties {soundWaveAssetPtr=&/Glider_AncientGladiatorMale_VOX_09}}
]
properties {
 attenuationSettings=/Game/Sounds/Attenuation/Fort_Player_Glider_OpenClose_Athena_Att
 audioPropertiesSheet=/BRCosmetics/Sounds/Settings/PropertySheets/Glider/APS_Cosmetic_Glider_Open
 maxDistance=10000
 soundClassObject=&Fort_Gliders_Land
 volumeMultiplier=1
 concurrencySet [&Athena_GliderOpenClose_Conc]
}
