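# Thrust loop sound cue for the AncientGladiatorMale glider cosmetic, mixing magic, trot and vocal layers.
# The magic loop's pitch and volume follow AnyDirection, while the trot loop's follow Forwards.
# Randomized, pitch-varied vocals repeat with delays after an initial wait, also modulated by Forwards.
# A quality-level switch gives the first two levels nine vocal variants and the last two only four.
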SoundCue /BRCosmetics/Sounds/Gliders/AncientGladiatorMale/Glider_AncientGladiatorMale_Thrust_Loop_Cue
root=Mixer_0
nodes [
 {id=Delay_0, children [Looping_0], properties {delayMax=6, delayMin=3}}
 {id=Delay_1, children [Random_2], properties {delayMax=3, delayMin=2}}
 {id=Delay_2, children [Looping_1], properties {delayMax=6, delayMin=3}}
 {id=Delay_3, children [Random_1], properties {delayMax=3, delayMin=2}}
 {id=Looping_0, children [Delay_1]}
 {id=Looping_1, children [Delay_3]}
 {id=Mixer_0, children [ModulatorContinuous_0, ModulatorContinuous_2, QualityLevel_0], properties {inputVolume [0.65,0.75,0.6]}}
 {
  id=ModulatorContinuous_0
  children [WavePlayer_0]
  properties {
   pitchModulationParams {default=0, maxOutput=1.25, minOutput=0.75, parameterName=AnyDirection}
   volumeModulationParams {default=0, minOutput=0.5, parameterName=AnyDirection}
  }
 }
 {
  id=ModulatorContinuous_1
  children [Modulator_1]
  properties {pitchModulationParams {default=0, minOutput=1, parameterName=Forwards}, volumeModulationParams {default=0, parameterName=Forwards}}
 }
 {
  id=ModulatorContinuous_2
  children [WavePlayer_1]
  properties {
   pitchModulationParams {default=0, maxOutput=1.25, minOutput=0.9, parameterName=Forwards}
   volumeModulationParams {default=0, minOutput=0.4, parameterName=Forwards}
  }
 }
 {
  id=ModulatorContinuous_3
  children [Modulator_0]
  properties {pitchModulationParams {default=0, minOutput=1, parameterName=Forwards}, volumeModulationParams {default=0, parameterName=Forwards}}
 }
 {id=Modulator_0, children [Delay_0], properties {pitchMax=1.2, pitchMin=1, volumeMax=1, volumeMin=0.6}}
 {id=Modulator_1, children [Delay_2], properties {pitchMax=1.2, pitchMin=1, volumeMax=1, volumeMin=0.6}}
 {id=QualityLevel_0, children [ModulatorContinuous_3, ModulatorContinuous_3, ModulatorContinuous_1, ModulatorContinuous_1]}
 {
  id=Random_1
  children [WavePlayer_2,WavePlayer_3,WavePlayer_4,WavePlayer_5]
  properties {bShouldExcludeFromBranchCulling=true, weights [1,1,1,1]}
 }
 {
  id=Random_2
  children [WavePlayer_38, WavePlayer_39, WavePlayer_40, WavePlayer_41, WavePlayer_42, WavePlayer_43, WavePlayer_44, WavePlayer_45, WavePlayer_46]
  properties {bShouldExcludeFromBranchCulling=true, weights [1,1,1,1,1,1,1,1,1]}
 }
 {
  id=WavePlayer_0
  properties {bLooping=true, soundWaveAssetPtr=&/Glider_AncientGladiatorMale_Thrust_Magic_Loop}
 }
 {
  id=WavePlayer_1
  properties {bLooping=true, soundWaveAssetPtr=&/Glider_AncientGladiatorMale_Thrust_Trot_Loop}
 }
 {id=WavePlayer_2, properties {soundWaveAssetPtr=&/Glider_AncientGladiatorMale_VOX_01}}
 {id=WavePlayer_3, properties {soundWaveAssetPtr=&/Glider_AncientGladiatorMale_VOX_02}}
 {id=WavePlayer_38, properties {soundWaveAssetPtr=&Glider_AncientGladiatorMale_VOX_01}}
 {id=WavePlayer_39, properties {soundWaveAssetPtr=&Glider_AncientGladiatorMale_VOX_02}}
 {id=WavePlayer_4, properties {soundWaveAssetPtr=&/Glider_AncientGladiatorMale_VOX_03}}
 {id=WavePlayer_40, properties {soundWaveAssetPtr=&Glider_AncientGladiatorMale_VOX_03}}
 {id=WavePlayer_41, properties {soundWaveAssetPtr=&/Glider_AncientGladiatorMale_VOX_04}}
 {id=WavePlayer_42, properties {soundWaveAssetPtr=&/Glider_AncientGladiatorMale_VOX_05}}
 {id=WavePlayer_43, properties {soundWaveAssetPtr=&/Glider_AncientGladiatorMale_VOX_06}}
 {id=WavePlayer_44, properties {soundWaveAssetPtr=&/Glider_AncientGladiatorMale_VOX_07}}
 {id=WavePlayer_45, properties {soundWaveAssetPtr=&/Glider_AncientGladiatorMale_VOX_08}}
 {id=WavePlayer_46, properties {soundWaveAssetPtr=&/Glider_AncientGladiatorMale_VOX_09}}
 {id=WavePlayer_5, properties {soundWaveAssetPtr=&Glider_AncientGladiatorMale_VOX_04}}
]
properties {
 attenuationSettings=&Fort_Player_Glider_Thrust_Athena_Att
 audioPropertiesSheet=&APS_Cosmetic_Glider_Loop
 maxDistance=12500
 soundClassObject=&Fort_Gliders_Land
 volumeMultiplier=0.6
}
